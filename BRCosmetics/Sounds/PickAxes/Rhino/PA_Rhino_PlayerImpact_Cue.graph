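# Sound cue for the Rhino pickaxe's player impact, played as a cosmetic melee hit sound.
# A Random node picks among five equally weighted impact waves and feeds a root Modulator left at its default settings.
# It is routed to the impacts sound class, uses shared pickaxe impact attenuation, property sheet and melee weapon concurrency assets, and primes on load.
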SoundCue /BRCosmetics/Sounds/PickAxes/Rhino/PA_Rhino_PlayerImpact_Cue
root=Modulator_0
nodes [
 {id=Modulator_0, children [Random_1]}
 {id=Random_1, children [WavePlayer_4,WavePlayer_5,WavePlayer_6,WavePlayer_7,WavePlayer_9], properties {weights [1,1,1,1,1]}}
 {id=WavePlayer_4, properties {soundWaveAssetPtr=&/PA_Rhino_PlayerImpact_01}}
 {id=WavePlayer_5, properties {soundWaveAssetPtr=&/PA_Rhino_PlayerImpact_02}}
 {id=WavePlayer_6, properties {soundWaveAssetPtr=&/PA_Rhino_PlayerImpact_03}}
 {id=WavePlayer_7, properties {soundWaveAssetPtr=&/PA_Rhino_PlayerImpact_04}}
 {id=WavePlayer_9, properties {soundWaveAssetPtr=&/PA_Rhino_PlayerImpact_05}}
]
properties {
 attenuationSettings=/Game/Sounds/Attenuation/Fort_Pickaxe_Impact_Att
 audioPropertiesSheet=/BRCosmetics/Sounds/Settings/PropertySheets/Pickaxe/APS_Cosmetic_Pickaxe_Impact
 bPrimeOnLoad=true
 maxDistance=4300
 soundClassObject=&Fort_Impacts
 volumeMultiplier=0.975
 attenuationOverrides {
  bAttenuateWithLPF=true
  distanceAlgorithm=Logarithmic
  falloffDistance=800
  lPFRadiusMax=800
  lPFRadiusMin=400
  attenuationShapeExtents {x=300}
 }
 concurrencyOverrides {bLimitToOwner=true, maxCount=2, resolutionRule=StopOldest}
 concurrencySet [&Player_Melee_Weapon_Conc]
}
