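# Sound cue for the Rhino pickaxe's ready sound, a melee weapon cosmetic.
# A mixer layers the ready wave with a quieter, randomly chosen swing wave from three equally weighted variations.
# A modulator adds slight volume variation before output at a low overall volume.
# It uses close-range attenuation, primes on load and overrides concurrency to prevent new instances at the limit.
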SoundCue /BRCosmetics/Sounds/PickAxes/Rhino/PA_Rhino_Ready_Cue
root=Modulator_0
nodes [
 {id=Mixer_0, children [WavePlayer_5,Random_0], properties {inputVolume [1,0.45]}}
 {id=Modulator_0, children [Mixer_0], properties {volumeMax=1, volumeMin=0.9}}
 {id=Random_0, children [WavePlayer_6,WavePlayer_7,WavePlayer_8], properties {weights [1,1,1]}}
 {id=WavePlayer_5, properties {soundWaveAssetPtr=&/PA_Rhino_Ready}}
 {id=WavePlayer_6, properties {soundWaveAssetPtr=&/PA_Rhino_Swing_02}}
 {id=WavePlayer_7, properties {soundWaveAssetPtr=&/PA_Rhino_Swing_03}}
 {id=WavePlayer_8, properties {soundWaveAssetPtr=&/PA_Rhino_Swing_01}}
]
properties {
 attenuationSettings=&Fort_Weapon_Projectile_Close_Athena_Att
 audioPropertiesSheet=/BRCosmetics/Sounds/Settings/PropertySheets/Pickaxe/APS_Cosmetic_Pickaxe_Ready
 bOverrideConcurrency=true
 bPrimeOnLoad=true
 maxDistance=4950
 soundClassObject=&Fort_WeaponMelee
 volumeMultiplier=0.2
 concurrencyOverrides {resolutionRule=PreventNew}
}
